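# Post-process AnimBP for the FruitFire Nudge female head that offsets upper eyelid and eyelash bones on both eyes.
# The AnimGraph chains three passes of additive bone-space Modify Bone nodes over the input pose, some with alpha interpolation.
# Each update remaps blink, upper-lid-down, squint and brow-down curves into per-side weight variables, with blink suppressing the others to limit stacking.
# No visible connection links those variables to the node alphas.
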AnimBlueprint /BRCosmetics/Characters/Player/Female/Medium/Heads/F_MED_FruitFire_Nudge_Head/Meshes/FruitFIre_Nudge_PostProcessAnimBP
target_skeleton=/BRCosmetics/Characters/Player/Female/Medium/Heads/F_MED_FruitFire_Nudge_Head/Meshes/CustomSkeleton/Fortnite_Base_Head_FruitFire_Nudge_Skeleton
graphs [
 AnimGraph {nodes [
   n1 = "Transform (Modify) Bone - Bone: R_eye_lid_upper_mid"(
    ComponentPose=Local To Component(LocalPose=Input Pose())
    Alpha=1
    Translation="-0.213,0.822,-0.497"
    Rotation="-14.290,4.871,-53.989"
    Scale="1,1,1"
   ) {
    rotationMode=BMM_Additive
    rotationSpace=BCS_BoneSpace
    translationMode=BMM_Additive
    translationSpace=BCS_BoneSpace
    alphaScaleBiasClamp {interpSpeedDecreasing=15, interpSpeedIncreasing=20}
    boneToModify {boneName=R_eye_lid_upper_mid}
    rotation=<rot -14.29 4.871 -53.989>
    translation=<-0.2129 0.8218 -0.4968>
   }
   n2 = "Transform (Modify) Bone - Bone: dyn_eyelash_r_01"(ComponentPose=@n1, Alpha=1, Translation="1.31,0.967,-1.6", Rotation="20.666,10.696,3.814", Scale="1,1,1") {
    rotationMode=BMM_Additive
    rotationSpace=BCS_BoneSpace
    translationMode=BMM_Additive
    translationSpace=BCS_BoneSpace
    alphaScaleBiasClamp {interpSpeedDecreasing=15, interpSpeedIncreasing=20}
    boneToModify {boneName=dyn_eyelash_r_01}
    rotation=<rot 20.666 10.696 3.814>
    translation=<1.31 0.967 -1.6>
   }
   n3 = "Transform (Modify) Bone - Bone: dyn_eyelash_r_02"(ComponentPose=@n2, Alpha=1, Translation="1.316,0.2,-1.7", Rotation="20.848,9.340,2.762", Scale="1,1,1") {
    rotationMode=BMM_Additive
    rotationSpace=BCS_BoneSpace
    translationMode=BMM_Additive
    translationSpace=BCS_BoneSpace
    alphaScaleBiasClamp {interpSpeedDecreasing=15, interpSpeedIncreasing=20}
    boneToModify {boneName=dyn_eyelash_r_02}
    rotation=<rot 20.848 9.34 2.762>
    translation=<1.316 0.2 -1.7>
   }
   n4 = "Transform (Modify) Bone - Bone: dyn_eyelash_r_03"(ComponentPose=@n3, Alpha=1, Translation="0.847,0.2,-1.6", Rotation="31.000,-0.000,0.000", Scale="1,1,1") {
    rotationMode=BMM_Additive
    rotationSpace=BCS_BoneSpace
    translationMode=BMM_Additive
    translationSpace=BCS_BoneSpace
    alphaScaleBiasClamp {interpSpeedDecreasing=15, interpSpeedIncreasing=20}
    boneToModify {boneName=dyn_eyelash_r_03}
    rotation=<rot 31 0 0>
    translation=<0.847 0.2 -1.6>
   }
   n5 = "Transform (Modify) Bone - Bone: dyn_eyelash_r_04"(ComponentPose=@n4, Alpha=1, Translation="0.8,0,-1.4", Rotation="21.000,-0.000,0.000", Scale="1,1,1") {
    rotationMode=BMM_Additive
    rotationSpace=BCS_BoneSpace
    translationMode=BMM_Additive
    translationSpace=BCS_BoneSpace
    alphaScaleBiasClamp {interpSpeedDecreasing=15, interpSpeedIncreasing=20}
    boneToModify {boneName=dyn_eyelash_r_04}
    rotation=<rot 21 0 0>
    translation {x=0.8, z=-1.4}
   }
   n6 = "Transform (Modify) Bone - Bone: L_eye_lid_upper_mid"(ComponentPose=@n5, Alpha=1, Translation="0.553,1.483,-0.573", Rotation="-1.676,-10.448,-60.145", Scale="1,1,1") {
    rotationMode=BMM_Additive
    rotationSpace=BCS_BoneSpace
    translationMode=BMM_Additive
    translationSpace=BCS_BoneSpace
    alphaScaleBiasClamp {interpSpeedDecreasing=15, interpSpeedIncreasing=20}
    boneToModify {boneName=L_eye_lid_upper_mid}
    rotation=<rot -1.676 -10.448 -60.145>
    translation=<0.553 1.483 -0.573>
   }
   n7 = "Transform (Modify) Bone - Bone: dyn_eyelash_l_01"(ComponentPose=@n6, Alpha=1, Translation="0.2,0,-0.75", Rotation="30.000,-0.000,-0.000", Scale="1,1,1") {
    rotationMode=BMM_Additive
    rotationSpace=BCS_BoneSpace
    translationMode=BMM_Additive
    translationSpace=BCS_BoneSpace
    alphaScaleBiasClamp {interpSpeedDecreasing=15, interpSpeedIncreasing=20}
    boneToModify {boneName=dyn_eyelash_l_01}
    rotation=<rot 30.0001 0 0>
    translation {x=0.2, z=-0.75}
   }
   n8 = "Transform (Modify) Bone - Bone: dyn_eyelash_l_02"(ComponentPose=@n7, Alpha=1, Translation="0.5,0,-0.6", Rotation="22,0,0", Scale="1,1,1") {
    rotationMode=BMM_Additive
    rotationSpace=BCS_BoneSpace
    translationMode=BMM_Additive
    translationSpace=BCS_BoneSpace
    alphaScaleBiasClamp {interpSpeedDecreasing=15, interpSpeedIncreasing=20}
    boneToModify {boneName=dyn_eyelash_l_02}
    rotation {pitch=22}
    translation {x=0.5, z=-0.6}
   }
   n9 = "Transform (Modify) Bone - Bone: dyn_eyelash_l_03"(ComponentPose=@n8, Alpha=1, Translation="0.6,0,-0.5", Rotation="30.000,0.000,-0.000", Scale="1,1,1") {
    rotationMode=BMM_Additive
    rotationSpace=BCS_BoneSpace
    translationMode=BMM_Additive
    translationSpace=BCS_BoneSpace
    alphaScaleBiasClamp {interpSpeedDecreasing=15, interpSpeedIncreasing=20}
    boneToModify {boneName=dyn_eyelash_l_03}
    rotation=<rot 30 0 0>
    translation {x=0.6, z=-0.5}
   }
   n10 = "Transform (Modify) Bone - Bone: dyn_eyelash_l_04"(ComponentPose=@n9, Alpha=1, Translation="0.6,0.542,-0.5", Rotation="20.000,0.000,-0.000", Scale="1,1,1") {
    rotationMode=BMM_Additive
    rotationSpace=BCS_BoneSpace
    translationMode=BMM_Additive
    translationSpace=BCS_BoneSpace
    alphaScaleBiasClamp {interpSpeedDecreasing=15, interpSpeedIncreasing=20}
    boneToModify {boneName=dyn_eyelash_l_04}
    rotation=<rot 20 0 0>
    translation=<0.6 0.542 -0.5>
   }
   n11 = "Transform (Modify) Bone - Bone: R_eye_lid_upper_mid"(ComponentPose=@n10, Alpha=1, Translation="0,0,0", Rotation="-5,0,-30", Scale="1,1,1") {
    rotationMode=BMM_Additive
    rotationSpace=BCS_BoneSpace
    translationMode=BMM_Additive
    translationSpace=BCS_BoneSpace
    alphaScaleBiasClamp {bClampResult=true, interpSpeedDecreasing=15, interpSpeedIncreasing=15}
    boneToModify {boneName=R_eye_lid_upper_mid}
    rotation {pitch=-5, roll=-30}
   }
   n12 = "Transform (Modify) Bone - Bone: dyn_eyelash_r_01"(ComponentPose=@n11, Alpha=1, Translation="-0.7,0,-0.75", Rotation="40,0,0", Scale="1,1,1") {
    rotationMode=BMM_Additive
    rotationSpace=BCS_BoneSpace
    translationMode=BMM_Additive
    translationSpace=BCS_BoneSpace
    alphaScaleBiasClamp {bClampResult=true, interpSpeedDecreasing=15, interpSpeedIncreasing=15}
    boneToModify {boneName=dyn_eyelash_r_01}
    rotation {pitch=40}
    translation {x=-0.7, z=-0.75}
   }
   n13 = "Transform (Modify) Bone - Bone: dyn_eyelash_r_02"(ComponentPose=@n12, Alpha=1, Translation="-0.6,0,-1.2", Rotation="40,0,0", Scale="1,1,1") {
    rotationMode=BMM_Additive
    rotationSpace=BCS_BoneSpace
    translationMode=BMM_Additive
    translationSpace=BCS_BoneSpace
    alphaScaleBiasClamp {bClampResult=true, interpSpeedDecreasing=15, interpSpeedIncreasing=15}
    boneToModify {boneName=dyn_eyelash_r_02}
    rotation {pitch=40}
    translation {x=-0.6, z=-1.2}
   }
   n14 = "Transform (Modify) Bone - Bone: dyn_eyelash_r_03"(ComponentPose=@n13, Alpha=1, Translation="-0.7,0,-0.4", Rotation="40,0,0", Scale="1,1,1") {
    rotationMode=BMM_Additive
    rotationSpace=BCS_BoneSpace
    translationMode=BMM_Additive
    translationSpace=BCS_BoneSpace
    alphaScaleBiasClamp {bClampResult=true, bInterpResult=true, interpSpeedDecreasing=30, interpSpeedIncreasing=30}
    boneToModify {boneName=dyn_eyelash_r_03}
    rotation {pitch=40}
    translation {x=-0.7, z=-0.4}
   }
   n15 = "Transform (Modify) Bone - Bone: dyn_eyelash_r_04"(ComponentPose=@n14, Alpha=1, Translation="-0.8,0,-0.7", Rotation="40,0,0", Scale="1,1,1") {
    rotationMode=BMM_Additive
    rotationSpace=BCS_BoneSpace
    translationMode=BMM_Additive
    translationSpace=BCS_BoneSpace
    alphaScaleBiasClamp {bClampResult=true, bInterpResult=true, interpSpeedDecreasing=30, interpSpeedIncreasing=30}
    boneToModify {boneName=dyn_eyelash_r_04}
    rotation {pitch=40}
    translation {x=-0.8, z=-0.7}
   }
   n16 = "Transform (Modify) Bone - Bone: L_eye_lid_upper_mid"(ComponentPose=@n15, Alpha=1, Translation="0,0,0", Rotation="5,0,-30", Scale="1,1,1") {
    rotationMode=BMM_Additive
    rotationSpace=BCS_BoneSpace
    translationMode=BMM_Additive
    translationSpace=BCS_BoneSpace
    alphaScaleBiasClamp {bClampResult=true, interpSpeedDecreasing=15, interpSpeedIncreasing=15}
    boneToModify {boneName=L_eye_lid_upper_mid}
    rotation {pitch=5, roll=-30}
   }
   n17 = "Transform (Modify) Bone - Bone: dyn_eyelash_l_01"(ComponentPose=@n16, Alpha=1, Translation="-1,0,-0.25", Rotation="40,0,0", Scale="1,1,1") {
    rotationMode=BMM_Additive
    rotationSpace=BCS_BoneSpace
    translationMode=BMM_Additive
    translationSpace=BCS_BoneSpace
    alphaScaleBiasClamp {bClampResult=true, interpSpeedDecreasing=15, interpSpeedIncreasing=15}
    boneToModify {boneName=dyn_eyelash_l_01}
    rotation {pitch=40}
    translation {x=-1, z=-0.25}
   }
   n18 = "Transform (Modify) Bone - Bone: dyn_eyelash_l_02"(ComponentPose=@n17, Alpha=1, Translation="-0.8,0,-1", Rotation="40,0,0", Scale="1,1,1") {
    rotationMode=BMM_Additive
    rotationSpace=BCS_BoneSpace
    translationMode=BMM_Additive
    translationSpace=BCS_BoneSpace
    alphaScaleBiasClamp {bClampResult=true, interpSpeedDecreasing=15, interpSpeedIncreasing=15}
    boneToModify {boneName=dyn_eyelash_l_02}
    rotation {pitch=40}
    translation {x=-0.8, z=-1}
   }
   n19 = "Transform (Modify) Bone - Bone: dyn_eyelash_l_03"(ComponentPose=@n18, Alpha=1, Translation="-0.8,0,-0.6", Rotation="40,0,0", Scale="1,1,1") {
    rotationMode=BMM_Additive
    rotationSpace=BCS_BoneSpace
    translationMode=BMM_Additive
    translationSpace=BCS_BoneSpace
    alphaScaleBiasClamp {bClampResult=true, bInterpResult=true, interpSpeedDecreasing=30, interpSpeedIncreasing=30}
    boneToModify {boneName=dyn_eyelash_l_03}
    rotation {pitch=40}
    translation {x=-0.8, z=-0.6}
   }
   n20 = "Transform (Modify) Bone - Bone: dyn_eyelash_l_04"(ComponentPose=@n19, Alpha=1, Translation="-0.9,0,-0.5", Rotation="40,0,0", Scale="1,1,1") {
    rotationMode=BMM_Additive
    rotationSpace=BCS_BoneSpace
    translationMode=BMM_Additive
    translationSpace=BCS_BoneSpace
    alphaScaleBiasClamp {bClampResult=true, bInterpResult=true, interpSpeedDecreasing=30, interpSpeedIncreasing=30}
    boneToModify {boneName=dyn_eyelash_l_04}
    rotation {pitch=40}
    translation {x=-0.9, z=-0.5}
   }
   n21 = "Transform (Modify) Bone - Bone: R_eye_lid_upper_mid"(ComponentPose=@n20, Alpha=1, Translation="0,0,0", Rotation="-11.569,7.686,-45.858", Scale="1,1,1") {
    rotationMode=BMM_Additive
    rotationSpace=BCS_BoneSpace
    translationMode=BMM_Additive
    translationSpace=BCS_BoneSpace
    alphaScaleBiasClamp {bInterpResult=true, interpSpeedDecreasing=30, interpSpeedIncreasing=20}
    boneToModify {boneName=R_eye_lid_upper_mid}
    rotation=<rot -11.569 7.686 -45.858>
   }
   n22 = "Transform (Modify) Bone - Bone: dyn_eyelash_r_01"(ComponentPose=@n21, Alpha=1, Translation="0.368,0.000,-0.761", Rotation="40.000,-0.000,-0.000", Scale="1,1,1") {
    rotationMode=BMM_Additive
    rotationSpace=BCS_BoneSpace
    translationMode=BMM_Additive
    translationSpace=BCS_BoneSpace
    alphaScaleBiasClamp {bInterpResult=true, interpSpeedDecreasing=30, interpSpeedIncreasing=20}
    boneToModify {boneName=dyn_eyelash_r_01}
    rotation=<rot 40 0 0>
    translation {x=0.368, z=-0.761}
   }
   n23 = "Transform (Modify) Bone - Bone: dyn_eyelash_r_02"(ComponentPose=@n22, Alpha=1, Translation="-0.066,0.000,-0.570", Rotation="40.000,-0.000,0.000", Scale="1,1,1") {
    rotationMode=BMM_Additive
    rotationSpace=BCS_BoneSpace
    translationMode=BMM_Additive
    translationSpace=BCS_BoneSpace
    alphaScaleBiasClamp {bInterpResult=true, interpSpeedDecreasing=30, interpSpeedIncreasing=20}
    boneToModify {boneName=dyn_eyelash_r_02}
    rotation=<rot 40 0 0>
    translation {x=-0.066, z=-0.57}
   }
   n24 = "Transform (Modify) Bone - Bone: dyn_eyelash_r_03"(ComponentPose=@n23, Alpha=1, Translation="-0.435,0.172,-0.246", Rotation="40.000,-5.000,0.000", Scale="1,1,1") {
    rotationMode=BMM_Additive
    rotationSpace=BCS_BoneSpace
    translationMode=BMM_Additive
    translationSpace=BCS_BoneSpace
    alphaScaleBiasClamp {bInterpResult=true, interpSpeedDecreasing=30, interpSpeedIncreasing=20}
    boneToModify {boneName=dyn_eyelash_r_03}
    rotation=<rot 40 -5 0>
    translation=<-0.4352 0.172 -0.2458>
   }
   n25 = "Transform (Modify) Bone - Bone: dyn_eyelash_r_04"(ComponentPose=@n24, Alpha=1, Translation="0.442,-0.000,-0.497", Rotation="40.000,-0.000,0.000", Scale="1,1,1") {
    rotationMode=BMM_Additive
    rotationSpace=BCS_BoneSpace
    translationMode=BMM_Additive
    translationSpace=BCS_BoneSpace
    alphaScaleBiasClamp {bInterpResult=true, interpSpeedDecreasing=30, interpSpeedIncreasing=20}
    boneToModify {boneName=dyn_eyelash_r_04}
    rotation=<rot 40 0 0>
    translation {x=0.442, z=-0.497}
   }
   n26 = "Transform (Modify) Bone - Bone: L_eye_lid_upper_mid"(ComponentPose=@n25, Alpha=1, Translation="0,0,0", Rotation="11.597,-7.643,-45.918", Scale="1,1,1") {
    rotationMode=BMM_Additive
    rotationSpace=BCS_BoneSpace
    translationMode=BMM_Additive
    translationSpace=BCS_BoneSpace
    alphaScaleBiasClamp {bInterpResult=true, interpSpeedDecreasing=30, interpSpeedIncreasing=20}
    boneToModify {boneName=L_eye_lid_upper_mid}
    rotation=<rot 11.597 -7.643 -45.918>
   }
   n27 = "Transform (Modify) Bone - Bone: dyn_eyelash_l_01"(ComponentPose=@n26, Alpha=1, Translation="0.133,0.000,-0.340", Rotation="40.000,0.000,0.000", Scale="1,1,1") {
    rotationMode=BMM_Additive
    rotationSpace=BCS_BoneSpace
    translationMode=BMM_Additive
    translationSpace=BCS_BoneSpace
    alphaScaleBiasClamp {bInterpResult=true, interpSpeedDecreasing=30, interpSpeedIncreasing=20}
    boneToModify {boneName=dyn_eyelash_l_01}
    rotation=<rot 40 0 0>
    translation=<0.1329 0 -0.3399>
   }
   n28 = "Transform (Modify) Bone - Bone: dyn_eyelash_l_02"(ComponentPose=@n27, Alpha=1, Translation="-0.651,-0.000,-0.558", Rotation="40.000,0.000,-0.000", Scale="1,1,1") {
    rotationMode=BMM_Additive
    rotationSpace=BCS_BoneSpace
    translationMode=BMM_Additive
    translationSpace=BCS_BoneSpace
    alphaScaleBiasClamp {bInterpResult=true, interpSpeedDecreasing=30, interpSpeedIncreasing=20}
    boneToModify {boneName=dyn_eyelash_l_02}
    rotation=<rot 40 0 0>
    translation {x=-0.651, z=-0.558}
   }
   n29 = "Transform (Modify) Bone - Bone: dyn_eyelash_l_03"(ComponentPose=@n28, Alpha=1, Translation="-0.620,-0.000,-0.232", Rotation="40.000,0.000,-0.000", Scale="1,1,1") {
    rotationMode=BMM_Additive
    rotationSpace=BCS_BoneSpace
    translationMode=BMM_Additive
    translationSpace=BCS_BoneSpace
    alphaScaleBiasClamp {bInterpResult=true, interpSpeedDecreasing=30, interpSpeedIncreasing=20}
    boneToModify {boneName=dyn_eyelash_l_03}
    rotation=<rot 40 0 0>
    translation {x=-0.62, z=-0.232}
   }
   n30 = "Transform (Modify) Bone - Bone: dyn_eyelash_l_04"(ComponentPose=@n29, Alpha=1, Translation="0.142,-0.000,-0.424", Rotation="40,0,0", Scale="1,1,1") {
    rotationMode=BMM_Additive
    rotationSpace=BCS_BoneSpace
    translationMode=BMM_Additive
    translationSpace=BCS_BoneSpace
    alphaScaleBiasClamp {bInterpResult=true, interpSpeedDecreasing=30, interpSpeedIncreasing=20}
    boneToModify {boneName=dyn_eyelash_l_04}
    rotation {pitch=40}
    translation {x=0.142, z=-0.424}
   }
   Output Pose(Result=Component To Local(ComponentPose=@n30))
  ]}
 EventGraph {nodes [
   Event BlueprintUpdateAnimation()
   > Set R_BlinkPose_Full(R_BlinkPose_Full:float=MapRangeClamped(Value=GetCurveValue(CurveName=r_blink_pose), InRangeA=0.1, InRangeB=0.9)) -> (Output_Get:float)
   > Set L_BlinkPose_Full(L_BlinkPose_Full:float=MapRangeClamped(Value=GetCurveValue(CurveName=l_blink_pose), InRangeA=0.1, InRangeB=0.9)) -> (Output_Get:float)
   > n1 = Sequence!(execute) -> (then_0, then_1, then_2:Exec, then_3:Exec)
   n2 = GetCurveValue(CurveName=r_blink_pose)
   > n3 = Branch(Condition="float > float"(A=@n2, B=0.1))
   > Set R_upperLidDownPose_Full(R_upperLidDownPose_Full:float=0) -> (Output_Get:float)
   > Set R_squintPose_Full(R_squintPose_Full:float=0) -> (Output_Get:float)
   > Set R_brow_down_pose(R_brow_down_pose:float=0) -> (Output_Get:float)
   n4 = Set R_upperLidDownPose_Full(execute=@n3.else, R_upperLidDownPose_Full=MapRangeClamped(Value=GetCurveValue(CurveName=R_upperLidDown_pose), InRangeA=0.15, InRangeB=0.9))
   > n5 = Branch(Condition="float > float"(A=@n2, B=0.1))
   > Set R_squintPose_Full(R_squintPose_Full=0)
   n6 = Branch(execute=@n5.else, Condition="float > float"(A=@n2))
   n7 = MapRangeClamped(Value=GetCurveValue(CurveName=R_squint_inner_pose), InRangeA=0.6, InRangeB=1)
   > n8 = Set R_squintPose_Full(R_squintPose_Full=float - float(A=@n7, B=float * float(A=@n2, B=1.2) {is=title K2Node_PromotableOperator}))
   n9 = Branch(execute=[@n11.then, @n8.then], Condition="float > float"(A=@n4))
   n10 = GetCurveValue(CurveName=R_brow_down_pose)
   n11 = Set R_squintPose_Full(execute=@n6.else, R_squintPose_Full=@n7)
   Set R_brow_down_pose(execute=[@n9.then, @n12.then], R_brow_down_pose=float - float(A=@n10, B=float * float(A=@n11, B=0.5)))
   n12 = Branch(execute=@n9.else, Condition="float > float"(A=@n11))
   n13 = GetCurveValue(CurveName=r_blink_pose)
   Set R_brow_down_pose(
    execute=@n12.else
    R_brow_down_pose=float * float(A=MapRangeClamped(Value=@n13, InRangeA=0, InRangeB=0.5, OutRangeA=1, OutRangeB=0), B=float - float(A=@n10, B=@n11))
   )
   n14 = GetCurveValue(CurveName=l_blink_pose)
   n15 = Branch(execute=@n1.then_1, Condition="float > float"(A=@n14, B=0.1))
   > Set L_upperLidDownPose_Full(L_upperLidDownPose_Full:float=0) -> (Output_Get:float)
   > Set L_squintPose_Full(L_squintPose_Full:float=0) -> (Output_Get:float)
   > Set L_brow_down_pose(L_brow_down_pose:float=0) -> (Output_Get:float)
   n16 = Set L_upperLidDownPose_Full(execute=@n15.else, L_upperLidDownPose_Full=MapRangeClamped(Value=GetCurveValue(CurveName=L_upperLidDown_pose), InRangeA=0.15, InRangeB=0.9))
   > n17 = Branch(Condition="float > float"(A=@n14, B=0.1))
   > Set L_squintPose_Full(L_squintPose_Full=0)
   n18 = Branch(execute=@n17.else, Condition="float > float"(A=@n14))
   n19 = MapRangeClamped(Value=GetCurveValue(CurveName=L_squint_inner_pose), InRangeA=0.6, InRangeB=1)
   > n20 = Set L_squintPose_Full(L_squintPose_Full=float - float(A=@n19, B=@n14))
   n21 = Branch(execute=[@n20.then, @n22.then], Condition="float > float"(A=@n16))
   > Set L_brow_down_pose(L_brow_down_pose=0)
   n22 = Set L_squintPose_Full(execute=@n18.else, L_squintPose_Full=@n19)
   n23 = Branch(execute=@n21.else, Condition="float > float"(A=@n22))
   n24 = GetCurveValue(CurveName=L_brow_down_pose)
   > Set L_brow_down_pose(L_brow_down_pose=float - float(A=@n24, B=float * float(A=@n22, B=0.5)))
   n25 = GetCurveValue(CurveName=l_blink_pose)
   Set L_brow_down_pose(
    execute=@n23.else
    L_brow_down_pose=float * float(A=MapRangeClamped(Value=@n25, InRangeA=0, InRangeB=0.5, OutRangeA=1, OutRangeB=0), B=float - float(A=@n24, B=@n22))
   )
   TryGetPawnOwner()
   float * float(A=MapRangeClamped(Value=@n25, InRangeA=0.2, InRangeB=1, OutRangeA=0.8, OutRangeB=0), B=GetCurveValue(CurveName=L_brow_down_pose))
   float * float(A=MapRangeClamped(Value=@n13, InRangeA=0.2, InRangeB=1, OutRangeA=0.8, OutRangeB=0), B=GetCurveValue(CurveName=R_brow_down_pose))
  ]}
]
variables [
 L_BlinkPose_Full {type=number}
 L_BlinkPose_Small {type=number}
 L_brow_down_pose {type=number}
 L_squintPose_Full {type=number}
 L_upperLidDownPose_Full {type=number}
 R_BlinkPose_Full {type=number}
 R_BlinkPose_Small {type=number}
 R_brow_down_pose {type=number}
 R_squintPose_Full {type=number}
 R_upperLidDownPose_Full {type=number}
]
